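# Anim Blueprint for the ShatterFly backpack that poses its wings by player state and makes the shards float.
# The wingPosition state machine picks Base, Dive or Fly poses from skydive, parachute and targeting flags, then caches the result.
# Outside skydiving, additive Fly and Base poses weighted by remapped 2D pawn speed are layered on top.
# The event graph computes three sine/cosine time offsets, applied as additive bone-space translations to groups of wing bones.
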AnimBlueprint /BRCosmetics/Accessories/FORT_Backpacks/F_MED_ShatterFly_Backpack/Meshes/F_MED_Shatter_Fly_Pack_Skeleton_AnimBP
target_skeleton=/Game/Accessories/FORT_Backpacks/F_MED_ShatterFly_Backpack/Meshes/F_MED_Shatter_Fly_Pack_Skeleton
graphs [
 AnimGraph {nodes [
   n1 = Get ShardFloatC() -> (ShardFloatC:Vector)
   n2 = Get ShardFloatB() -> (ShardFloatB:Vector)
   n3 = Get PawnSpeedRemap() -> (PawnSpeedRemap:double)
   n4 = F_MED_Shatter_Fly_Pack_Poses_ADD_PoseAsset : Fly!(PoseWeight:float=@n3) -> (Pose:Pose Link Structure) {
    is=title AnimGraphNode_PoseByName
    poseAsset=&/F_MED_Shatter_Fly_Pack_Poses_ADD_PoseAsset
    poseName=Fly
   }
   n5 = F_MED_Shatter_Fly_Pack_Poses_ADD_PoseAsset : Base!(PoseWeight:float=@n3) -> (Pose:Pose Link Structure) {
    is=title AnimGraphNode_PoseByName
    poseAsset=&F_MED_Shatter_Fly_Pack_Poses_ADD_PoseAsset
    poseName=Base
   }
   n6 = Apply Additive(Base=Apply Additive(Base=Use cached pose 'preSprintCache'(), Additive=@n4) {alphaBoolBlend {blendOption=Linear}}, Additive=@n5) {alphaBoolBlend {blendOption=Linear}}
   n7 = Use cached pose 'preSprintCache'()
   n8 = Get bIsSkyDiving()
   n9 = Get ShardFloatA() -> (ShardFloatA:Vector)
   n10 = "Transform (Modify) Bone - Bone: upWing_02_l"(
    ComponentPose=Local To Component(LocalPose=Blend Poses by bool(BlendPose_0=@n7, BlendPose_1=@n6, BlendTime_0=0.3, BlendTime_1=0.3, bActiveValue=@n8) {blendType=Linear})
    Alpha=1
    Translation=@n9
    Rotation="0,0,0"
    Scale="1,1,1"
   ) {translationMode=BMM_Additive, translationSpace=BCS_BoneSpace, alphaBoolBlend {blendOption=Linear}, boneToModify {boneName=upWing_02_l}}
   n11 = "Transform (Modify) Bone - Bone: upWing_04_l"(ComponentPose=@n10, Alpha=1, Translation=@n9, Rotation="0,0,0", Scale="1,1,1") {translationMode=BMM_Additive, translationSpace=BCS_BoneSpace, alphaBoolBlend {blendOption=Linear}, boneToModify {boneName=upWing_04_l}}
   n12 = "Transform (Modify) Bone - Bone: upWing_02_r"(ComponentPose=@n11, Alpha=1, Translation=@n9, Rotation="0,0,0", Scale="1,1,1") {translationMode=BMM_Additive, translationSpace=BCS_BoneSpace, alphaBoolBlend {blendOption=Linear}, boneToModify {boneName=upWing_02_r}}
   n13 = "Transform (Modify) Bone - Bone: downWing_01_l"(ComponentPose=@n12, Alpha=1, Translation=@n9, Rotation="0,0,0", Scale="1,1,1") {translationMode=BMM_Additive, translationSpace=BCS_BoneSpace, alphaBoolBlend {blendOption=Linear}, boneToModify {boneName=downWing_01_l}}
   n14 = "Transform (Modify) Bone - Bone: downWing_03_l"(ComponentPose=@n13, Alpha=1, Translation=@n9, Rotation="0,0,0", Scale="1,1,1") {translationMode=BMM_Additive, translationSpace=BCS_BoneSpace, alphaBoolBlend {blendOption=Linear}, boneToModify {boneName=downWing_03_l}}
   n15 = "Transform (Modify) Bone - Bone: downWing_05_r"(ComponentPose=@n14, Alpha=1, Translation=@n9, Rotation="0,0,0", Scale="1,1,1") {translationMode=BMM_Additive, translationSpace=BCS_BoneSpace, alphaBoolBlend {blendOption=Linear}, boneToModify {boneName=downWing_05_r}}
   n16 = "Transform (Modify) Bone - Bone: upWing_03_l"(ComponentPose=@n15, Alpha=1, Translation=@n2, Rotation="0,0,0", Scale="1,1,1") {translationMode=BMM_Additive, translationSpace=BCS_BoneSpace, alphaBoolBlend {blendOption=Linear}, boneToModify {boneName=upWing_03_l}}
   n17 = "Transform (Modify) Bone - Bone: upWing_05_l"(ComponentPose=@n16, Alpha=1, Translation=@n2, Rotation="0,0,0", Scale="1,1,1") {translationMode=BMM_Additive, translationSpace=BCS_BoneSpace, alphaBoolBlend {blendOption=Linear}, boneToModify {boneName=upWing_05_l}}
   n18 = "Transform (Modify) Bone - Bone: upWing_03_r"(ComponentPose=@n17, Alpha=1, Translation=@n2, Rotation="0,0,0", Scale="1,1,1") {translationMode=BMM_Additive, translationSpace=BCS_BoneSpace, alphaBoolBlend {blendOption=Linear}, boneToModify {boneName=upWing_03_r}}
   n19 = "Transform (Modify) Bone - Bone: downWing_02_l"(ComponentPose=@n18, Alpha=1, Translation=@n2, Rotation="0,0,0", Scale="1,1,1") {translationMode=BMM_Additive, translationSpace=BCS_BoneSpace, alphaBoolBlend {blendOption=Linear}, boneToModify {boneName=downWing_02_l}}
   n20 = "Transform (Modify) Bone - Bone: downWing_03_r"(ComponentPose=@n19, Alpha=1, Translation=@n2, Rotation="0,0,0", Scale="1,1,1") {translationMode=BMM_Additive, translationSpace=BCS_BoneSpace, alphaBoolBlend {blendOption=Linear}, boneToModify {boneName=downWing_03_r}}
   n21 = "Transform (Modify) Bone - Bone: downWing_01_r"(ComponentPose=@n20, Alpha=1, Translation=@n2, Rotation="0,0,0", Scale="1,1,1") {translationMode=BMM_Additive, translationSpace=BCS_BoneSpace, alphaBoolBlend {blendOption=Linear}, boneToModify {boneName=downWing_01_r}}
   n22 = "Transform (Modify) Bone - Bone: upWing_01_l"(ComponentPose=@n21, Alpha=1, Translation=@n1, Rotation="0,0,0", Scale="1,1,1") {translationMode=BMM_Additive, translationSpace=BCS_BoneSpace, alphaBoolBlend {blendOption=Linear}, boneToModify {boneName=upWing_01_l}}
   n23 = "Transform (Modify) Bone - Bone: upWing_01_r"(ComponentPose=@n22, Alpha=1, Translation=@n1, Rotation="0,0,0", Scale="1,1,1") {translationMode=BMM_Additive, translationSpace=BCS_BoneSpace, alphaBoolBlend {blendOption=Linear}, boneToModify {boneName=upWing_01_r}}
   n24 = "Transform (Modify) Bone - Bone: downWing_04_l"(ComponentPose=@n23, Alpha=1, Translation=@n1, Rotation="0,0,0", Scale="1,1,1") {translationMode=BMM_Additive, translationSpace=BCS_BoneSpace, alphaBoolBlend {blendOption=Linear}, boneToModify {boneName=downWing_04_l}}
   n25 = "Transform (Modify) Bone - Bone: downWing_02_r"(ComponentPose=@n24, Alpha=1, Translation=@n1, Rotation="0,0,0", Scale="1,1,1") {translationMode=BMM_Additive, translationSpace=BCS_BoneSpace, alphaBoolBlend {blendOption=Linear}, boneToModify {boneName=downWing_02_r}}
   n26 = "Transform (Modify) Bone - Bone: downWing_04_r"(ComponentPose=@n25, Alpha=1, Translation=@n1, Rotation="0,0,0", Scale="1,1,1") {translationMode=BMM_Additive, translationSpace=BCS_BoneSpace, alphaBoolBlend {blendOption=Linear}, boneToModify {boneName=downWing_04_r}}
   Output Pose(Result=Component To Local(ComponentPose=@n26)) {layerGroup=None}
   AnimGraphNode_StateMachine_0 = wingPosition() {is=title AnimGraphNode_StateMachine}
   Save cached pose 'preSprintCache'(Pose=@AnimGraphNode_StateMachine_0) {cacheName=preSprintCache}
   AnimGraphNode_SubInput_0 = Input Pose()
  ]}
 wingPosition {
  entry=Base
  path=AnimGraph.AnimGraphNode_StateMachine_0.wingPosition
  ids=id
  nodes [
   AnimStateNode_0 = Base() {is=title AnimStateNode}
   AnimStateNode_1 = Diving() {is=title AnimStateNode}
   AnimStateNode_2 = Floating() {is=title AnimStateNode}
   AnimStateNode_3 = Targeting() {is=title AnimStateNode}
   AnimStateTransitionNode_0 = AnimStateTransitionNode()
   AnimStateTransitionNode_1 = AnimStateTransitionNode()
   AnimStateTransitionNode_2 = AnimStateTransitionNode()
   AnimStateTransitionNode_3 = AnimStateTransitionNode()
   AnimStateTransitionNode_5 = AnimStateTransitionNode()
   AnimStateTransitionNode_6 = AnimStateTransitionNode()
   AnimStateTransitionNode_7 = AnimStateTransitionNode()
  ]
  links [
   {
    dst=Diving
    rule=AnimGraph.AnimGraphNode_StateMachine_0.wingPosition.AnimStateTransitionNode_0.Transition
    src=Base
    via=AnimStateTransitionNode_0
   }
   {
    dst=Targeting
    rule=AnimGraph.AnimGraphNode_StateMachine_0.wingPosition.AnimStateTransitionNode_6.Transition
    src=Base
    via=AnimStateTransitionNode_6
   }
   {
    dst=Base
    rule=AnimGraph.AnimGraphNode_StateMachine_0.wingPosition.AnimStateTransitionNode_3.Transition
    src=Diving
    via=AnimStateTransitionNode_3
   }
   {
    dst=Floating
    rule=AnimGraph.AnimGraphNode_StateMachine_0.wingPosition.AnimStateTransitionNode_1.Transition
    src=Diving
    via=AnimStateTransitionNode_1
   }
   {
    dst=Base
    rule=AnimGraph.AnimGraphNode_StateMachine_0.wingPosition.AnimStateTransitionNode_2.Transition
    src=Floating
    via=AnimStateTransitionNode_2
   }
   {
    dst=Diving
    rule=AnimGraph.AnimGraphNode_StateMachine_0.wingPosition.AnimStateTransitionNode_5.Transition
    src=Floating
    via=AnimStateTransitionNode_5
   }
   {
    dst=Base
    rule=AnimGraph.AnimGraphNode_StateMachine_0.wingPosition.AnimStateTransitionNode_7.Transition
    src=Targeting
    via=AnimStateTransitionNode_7
   }
  ]
 }
 Base {
  path=AnimGraph.AnimGraphNode_StateMachine_0.wingPosition.AnimStateNode_0.Base
  nodes [
   n1 = F_MED_Shatter_Fly_Pack_Poses_PoseAsset : Base!(PoseWeight:float=1) -> (Pose:Pose Link Structure) {
    is=title AnimGraphNode_PoseByName
    poseAsset=&/F_MED_Shatter_Fly_Pack_Poses_PoseAsset
    poseName=Base
   }
   Output Animation Pose(Result=@n1)
  ]
 }
 Diving {
  path=AnimGraph.AnimGraphNode_StateMachine_0.wingPosition.AnimStateNode_1.Diving
  nodes [
   n1 = F_MED_Shatter_Fly_Pack_Poses_PoseAsset : Dive!(PoseWeight:float=1) -> (Pose:Pose Link Structure) {
    is=title AnimGraphNode_PoseByName
    poseAsset=&F_MED_Shatter_Fly_Pack_Poses_PoseAsset
    poseName=Dive
   }
   Output Animation Pose(Result=@n1)
  ]
 }
 Floating {
  path=AnimGraph.AnimGraphNode_StateMachine_0.wingPosition.AnimStateNode_2.Floating
  nodes [
   n1 = F_MED_Shatter_Fly_Pack_Poses_PoseAsset : Fly!(PoseWeight:float=1) -> (Pose:Pose Link Structure) {
    is=title AnimGraphNode_PoseByName
    poseAsset=&F_MED_Shatter_Fly_Pack_Poses_PoseAsset
    poseName=Fly
   }
   Output Animation Pose(Result=@n1)
  ]
 }
 Targeting {
  path=AnimGraph.AnimGraphNode_StateMachine_0.wingPosition.AnimStateNode_3.Targeting
  nodes [
   n1 = F_MED_Shatter_Fly_Pack_Poses_PoseAsset : Fly(PoseWeight=1) {poseAsset=&F_MED_Shatter_Fly_Pack_Poses_PoseAsset, poseName=Fly}
   Output Animation Pose(Result=@n1)
  ]
 }
 Transition {
  path=AnimGraph.AnimGraphNode_StateMachine_0.wingPosition.AnimStateTransitionNode_0.Transition
  nodes [Result(bCanEnterTransition=Get bIsSkydiveDiveMode())]
 }
 Transition {
  path=AnimGraph.AnimGraphNode_StateMachine_0.wingPosition.AnimStateTransitionNode_1.Transition
  nodes [Result(bCanEnterTransition=Get bIsSkydiveFloating())]
 }
 Transition {
  path=AnimGraph.AnimGraphNode_StateMachine_0.wingPosition.AnimStateTransitionNode_2.Transition
  nodes [Result(bCanEnterTransition=Get bIsParachuteOpened())]
 }
 Transition {
  path=AnimGraph.AnimGraphNode_StateMachine_0.wingPosition.AnimStateTransitionNode_3.Transition
  nodes [Result(bCanEnterTransition=Get bIsParachuteOpened())]
 }
 Transition {
  path=AnimGraph.AnimGraphNode_StateMachine_0.wingPosition.AnimStateTransitionNode_5.Transition
  nodes [Result(bCanEnterTransition=Get bIsSkydiveDiveMode())]
 }
 Transition {
  path=AnimGraph.AnimGraphNode_StateMachine_0.wingPosition.AnimStateTransitionNode_6.Transition
  nodes [Result(bCanEnterTransition=Get bIsTargeting())]
 }
 Transition {
  path=AnimGraph.AnimGraphNode_StateMachine_0.wingPosition.AnimStateTransitionNode_7.Transition
  nodes [Result(bCanEnterTransition=NOT Boolean(A=Get bIsTargeting()))]
 }
 EventGraph {nodes [
   Event BlueprintUpdateAnimation()
   > Set PawnSpeedRemap(PawnSpeedRemap:double=MapRangeClamped(Value=Get PawnSpeed2D(), InRangeA=0, InRangeB=549)) -> (Output_Get:double)
   > Sequence()
   n1 = GetGameTimeInSeconds()
   n2 = "Sin (Radians)"(A=float * float(A=float * float(A=1, B=1), B=@n1))
   n3 = "Sin (Radians)"(A=float * float(A=float * float(A=1, B=2), B=@n1))
   n4 = "Sin (Radians)"(A=float * float(A=float * float(A=1, B=1.5), B=@n1))
   > Set ShardFloatA(ShardFloatA:Vector=MakeVector(X=float * float(A=@n4, B=0.3), Y=float * float(A=@n2, B=0.56), Z=float * float(A=@n3, B=0.22))) -> (Output_Get:Vector)
   n5 = "Cos (Radians)"!(A:double=float * float(A=float * float(A=1, B=1.5), B=@n1)) -> (ReturnValue:double)
   n6 = "Cos (Radians)"(A=float * float(A=float * float(A=1, B=1.78), B=@n1))
   n7 = "Cos (Radians)"(A=float * float(A=float * float(A=1, B=1.15), B=@n1))
   > Set ShardFloatB(ShardFloatB:Vector=MakeVector(X=float * float(A=@n5, B=0.6), Y=float * float(A=@n6, B=0.725), Z=float * float(A=@n7, B=0.3))) -> (Output_Get:Vector)
   n8 = "Sin (Radians)"(A=float * float(A=float * float(A=1, B=1.5), B=@n1))
   n9 = "Sin (Radians)"(A=float * float(A=float * float(A=1, B=2), B=@n1))
   n10 = "Sin (Radians)"(A=float * float(A=float * float(A=1, B=1), B=@n1))
   > Set ShardFloatC(ShardFloatC:Vector=MakeVector(X=float * float(A=@n8, B=0.3), Y=float * float(A=@n10, B=0.6), Z=float * float(A=@n9, B=0.25))) -> (Output_Get:Vector)
   TryGetPawnOwner()
  ]}
]
variables [
 PawnSpeedRemap {type=number}
 ShardFloatA {type=Vector}
 ShardFloatB {type=Vector}
 ShardFloatC {type=Vector}
]
